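# Niagara system for the Possession pickaxe's animation trails, drawing two short-lived ribbons, SoftTrail and TipTrail.
# Both emitters use shared FN anim-trail modules to spawn and update particles from the user skeletal mesh's bones, adding a user-supplied velocity.
# Time-of-day selectors pick color-over-life curves that fade alpha out, starting fainter at night.
# SoftTrail also selects its initial color by time of day and writes dynamic material parameters.
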NiagaraSystem /BRCosmetics/Weapons/FORT_Melee/Pickaxe_Possession/FX/NS_Pickaxe_Possession_Animtrails
dependencies {usedDataInterfaces [&NiagaraDataInterfaceColorCurve], usedDynamicInputs [], usedModules [&SystemState]}
emitters {
 SoftTrail {
  emitterUpdateScript [
   &EmitterState()
   &FN_SpawnAnimTrail()
  ]
  particleSpawnScript [
   &FN_InitializeAnimTrailBoneIndecies()
   &FN_InitializeAnimTrail(
    AnimTrailWidth=1.5
    Color=&FN_Select_Color_from_Time_Of_Day(Morning=<rgba 0 2.8891 4 1>, Day=<rgba 0 3.8853 5 1>, Evening=<rgba 0 3.5105 4 1>, Night=<rgba 0 3.8978 5 1>)
    Lifetime=0.2
   )
   &AddVelocity(Velocity=$User.NewVector3f)
  ]
  particleUpdateScript [
   &FN_UpdateAnimTrail(SkeletalMesh=$User.SkeletalMesh) {module=FN_UpdateAnimTrail001}
   &Color(Color=&FN_Select_Color_from_Time_Of_Day(
     Morning=&ColorFromCurve(ColorCurve=DataInterface() {propertyValues {
        AlphaCurve {keys [<at 0 1 RCIM_Cubic>,<at 0.5541 0 RCIM_Cubic>]}
        BlueCurve {keys [<at 0 0.151>,<at 0.9673 0.3021>]}
        GreenCurve {keys [<at 0 0.151>,<at 0.9673 0.236>]}
        RedCurve {keys [<at 0 0.151>,<at 0.9673 0.2456>]}
       }})
     Day=&ColorFromCurve(ColorCurve=DataInterface() {propertyValues {
        AlphaCurve {keys [<at 0 1 RCIM_Cubic>,<at 0.5541 0 RCIM_Cubic>]}
        BlueCurve {keys [<at 0 0.151>,<at 0.9673 0.3021>]}
        GreenCurve {keys [<at 0 0.151>,<at 0.9673 0.236>]}
        RedCurve {keys [<at 0 0.151>,<at 0.9673 0.2456>]}
       }})
     Evening=&ColorFromCurve(ColorCurve=DataInterface() {propertyValues {
        AlphaCurve {keys [<at 0 1 RCIM_Cubic>,<at 0.5541 0 RCIM_Cubic>]}
        BlueCurve {keys [<at 0 0.151>,<at 0.9673 0.3021>]}
        GreenCurve {keys [<at 0 0.151>,<at 0.9673 0.236>]}
        RedCurve {keys [<at 0 0.151>,<at 0.9673 0.2456>]}
       }})
     Night=&ColorFromCurve(ColorCurve=DataInterface() {propertyValues {
        AlphaCurve {keys [<at 0 0.25 RCIM_Cubic>,<at 0.5541 0 RCIM_Cubic>]}
        BlueCurve {keys [<at 0 0.0433>,<at 0.9673 0.3021>]}
        GreenCurve {keys [<at 0 0.12>,<at 0.9673 0.236>]}
        RedCurve {keys [<at 0 0.151>,<at 0.9673 0.2456>]}
       }})
    ))
   &SolveForcesAndVelocity(Clamp Velocity=false, Limit Acceleration=false)
   &DynamicMaterialParameters(Module.Param0WriteEnabled=true, Module.Param1WriteEnabled=true, Module.Param2WriteEnabled=true, Module.Param3WriteEnabled=true)
  ]
  rendererClasses [&NiagaraRibbonRendererProperties]
  renderers [{rendererClass=&NiagaraRibbonRendererProperties}]
 }
 TipTrail {
  emitterUpdateScript [
   &EmitterState()
   &FN_SpawnAnimTrail()
  ]
  particleSpawnScript [
   &FN_InitializeAnimTrailBoneIndecies()
   &FN_InitializeAnimTrail(Lifetime=0.2)
   &AddVelocity(Velocity=$User.NewVector3f)
  ]
  particleUpdateScript [
   &FN_UpdateAnimTrail(SkeletalMesh=$User.SkeletalMesh) {module=FN_UpdateAnimTrail001}
   &Color(Color=&FN_Select_Color_from_Time_Of_Day(
     Morning=&ColorFromCurve(ColorCurve=DataInterface() {propertyValues {
        AlphaCurve {keys [<at 0 1>,<at 0.3291 0 RCIM_Cubic>]}
        BlueCurve {keys [<at 0.0011 0.4>,<at 0.1935 0.0167>]}
        GreenCurve {keys [<at 0.0011 0.6389>,<at 0.1935 0.056>]}
        RedCurve {keys [<at 0.0011 0.8>,<at 0.1935 0.1146>]}
       }})
     Day=&ColorFromCurve(ColorCurve=DataInterface() {propertyValues {
        AlphaCurve {keys [<at 0 1>,<at 0.3291 0 RCIM_Cubic>]}
        BlueCurve {keys [<at 0.0011 0.4>,<at 0.1935 0.0167>]}
        GreenCurve {keys [<at 0.0011 0.6389>,<at 0.1935 0.056>]}
        RedCurve {keys [<at 0.0011 0.8>,<at 0.1935 0.1146>]}
       }})
     Evening=&ColorFromCurve(ColorCurve=DataInterface() {propertyValues {
        AlphaCurve {keys [<at 0 1>,<at 0.3291 0 RCIM_Cubic>]}
        BlueCurve {keys [<at 0.0011 0.4>,<at 0.1935 0.0167>]}
        GreenCurve {keys [<at 0.0011 0.6389>,<at 0.1935 0.056>]}
        RedCurve {keys [<at 0.0011 0.8>,<at 0.1935 0.1146>]}
       }})
     Night=&ColorFromCurve(ColorCurve=DataInterface() {propertyValues {
        AlphaCurve {keys [<at 0 0.65>,<at 0.3291 0 RCIM_Cubic>]}
        BlueCurve {keys [<at 0.0011 0.4>,<at 0.1935 0.0167>]}
        GreenCurve {keys [<at 0.0011 0.6389>,<at 0.1935 0.056>]}
        RedCurve {keys [<at 0.0011 0.8>,<at 0.1935 0.1146>]}
       }})
    ))
   &SolveForcesAndVelocity(Clamp Velocity=false, Limit Acceleration=false)
  ]
  rendererClasses [&NiagaraRibbonRendererProperties]
  renderers [{rendererClass=&NiagaraRibbonRendererProperties}]
 }
}
summary {
 systemName=NS_Pickaxe_Possession_Animtrails
 emitters [
  {emitterName=SoftTrail, rendererClasses [&NiagaraRibbonRendererProperties]}
  {emitterName=TipTrail, rendererClasses [&NiagaraRibbonRendererProperties]}
 ]
 userVariables [
  User.Color A {type=LinearColor, description="", defaultValue {struct=LinearColor, value=<rgba 1 1 1 1>}}
  User.NewVector3f {type=Vector3f, description="", defaultValue {struct=Vector3f, value=<0 0 0>}}
  User.RibbonIDOffset {type=NiagaraInt32}
  User.SkeletalMesh {
   type=NiagaraDataInterfaceSkeletalMesh
   defaultValue {
    struct=NiagaraExt_VariableValue_DataInterface
    value {dataInterface=&self:NiagaraDataInterfaceSkeletalMesh_0}
   }
  }
  User.ToonRibbonPanSpeed {type=NiagaraFloat, defaultValue=-2, description=""}
  User.VariantPicker {type=NiagaraFloat, defaultValue=0, description=""}
 ]
}
